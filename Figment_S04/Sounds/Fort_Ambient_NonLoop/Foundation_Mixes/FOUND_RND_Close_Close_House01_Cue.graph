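# Sound cue for non-looping house creak ambience in the Foundations ambience class, spatialized with a map attenuation preset.
# The first quality slot mixes random building creaks with delayed floor creaks and quieter, delayed stereo house creaks, each randomized in pitch and volume.
# The second slot plays just one of three modulated creaks, a lighter variant; the other two slots are empty.
# A concurrency override caps simultaneous instances and blocks new ones.
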SoundCue /Figment_S04/Sounds/Fort_Ambient_NonLoop/Foundation_Mixes/FOUND_RND_Close_Close_House01_Cue
root=QualityLevel_0
nodes [
 {id=Delay_20, children [Modulator_1], properties {delayMax=5, delayMin=4}}
 {id=Delay_21, children [Modulator_7], properties {delayMax=2, delayMin=1}}
 {id=Mixer_16, children [Delay_20,Modulator_4,Delay_21], properties {inputVolume [0.7,1,0.15]}}
 {id=Modulator_1, children [Random_49], properties {pitchMax=0.9, pitchMin=0.6, volumeMax=1.1, volumeMin=0.9}}
 {id=Modulator_2, children [WavePlayer_2], properties {pitchMax=0.9, pitchMin=0.6, volumeMax=0.8, volumeMin=0.6}}
 {id=Modulator_4, children [Random_48], properties {pitchMax=1.2, pitchMin=1, volumeMax=1.1, volumeMin=0.9}}
 {id=Modulator_5, children [WavePlayer_5], properties {pitchMax=1.2, pitchMin=1, volumeMax=1.1, volumeMin=0.9}}
 {id=Modulator_55, children [Mixer_16], properties {pitchMax=1.1, pitchMin=0.9, volumeMax=1.1, volumeMin=0.9}}
 {id=Modulator_7, children [Random_50], properties {pitchMax=0.8, pitchMin=0.5, volumeMax=1.5, volumeMin=0.3}}
 {id=Modulator_8, children [WavePlayer_8], properties {pitchMax=0.8, pitchMin=0.5, volumeMax=0.5, volumeMin=0.15}}
 {id=QualityLevel_0, children [Modulator_55,Random_0,null,null]}
 {id=Random_0, children [Modulator_2,Modulator_5,Modulator_8], properties {weights [1,1,1]}}
 {id=Random_48, children [WavePlayer_4, WavePlayer_302, WavePlayer_303], properties {weights [1,1,1]}}
 {id=Random_49, children [WavePlayer_304, WavePlayer_1, WavePlayer_306, WavePlayer_307, WavePlayer_308], properties {weights [1,1,1,1,1]}}
 {
  id=Random_50
  children [WavePlayer_309, WavePlayer_310, WavePlayer_311, WavePlayer_312, WavePlayer_7, WavePlayer_314]
  properties {weights [1,1,1,1,1,1]}
 }
 {id=WavePlayer_1, properties {soundWaveAssetPtr=/Figment_S04/Sounds/Trans_AmbientNonLoop/Wood/HouseFloor_CreakMediumA001}}
 {id=WavePlayer_2, properties {soundWaveAssetPtr=&HouseFloor_CreakMediumA001}}
 {id=WavePlayer_302, properties {soundWaveAssetPtr=/Figment_S04/Sounds/Trans_AmbientNonLoop/Wood/BuildingCreaksA002}}
 {id=WavePlayer_303, properties {soundWaveAssetPtr=/Figment_S04/Sounds/Trans_AmbientNonLoop/Wood/BuildingCreaksA003}}
 {id=WavePlayer_304, properties {soundWaveAssetPtr=/Figment_S04/Sounds/Trans_AmbientNonLoop/Wood/HouseFloor_CreakMediumA002}}
 {id=WavePlayer_306, properties {soundWaveAssetPtr=/Figment_S04/Sounds/Trans_AmbientNonLoop/Wood/HouseFloor_CreakMediumA003}}
 {id=WavePlayer_307, properties {soundWaveAssetPtr=/Figment_S04/Sounds/Trans_AmbientNonLoop/Wood/HouseFloor_CreakMediumA004}}
 {id=WavePlayer_308, properties {soundWaveAssetPtr=/Figment_S04/Sounds/Trans_AmbientNonLoop/Wood/HouseFloor_CreakMediumA005}}
 {id=WavePlayer_309, properties {soundWaveAssetPtr=/Game/Sounds/Trans_AmbientNonLoop/Wood/CreakHouseA001_Stereo}}
 {id=WavePlayer_310, properties {soundWaveAssetPtr=/Figment_S04/Sounds/Trans_AmbientNonLoop/Wood/CreakHouseA002_Stereo}}
 {id=WavePlayer_311, properties {soundWaveAssetPtr=/Figment_S04/Sounds/Trans_AmbientNonLoop/Wood/CreakHouseA003_Stereo}}
 {id=WavePlayer_312, properties {soundWaveAssetPtr=/Figment_S04/Sounds/Trans_AmbientNonLoop/Wood/CreakHouseA004_Stereo}}
 {id=WavePlayer_314, properties {soundWaveAssetPtr=/Figment_S04/Sounds/Trans_AmbientNonLoop/Wood/CreakHouseA006_Stereo}}
 {id=WavePlayer_4, properties {soundWaveAssetPtr=/Figment_S04/Sounds/Trans_AmbientNonLoop/Wood/BuildingCreaksA001}}
 {id=WavePlayer_5, properties {soundWaveAssetPtr=&BuildingCreaksA001}}
 {id=WavePlayer_7, properties {soundWaveAssetPtr=/Figment_S04/Sounds/Trans_AmbientNonLoop/Wood/CreakHouseA005_Stereo}}
 {id=WavePlayer_8, properties {soundWaveAssetPtr=&CreakHouseA005_Stereo}}
]
properties {
 attenuationSettings=/Game/Sounds/Attenuation/Fort_Map_3x3_Square
 bOverrideConcurrency=true
 maxDistance=1780
 soundClassObject=/Figment_S04/Sounds/FortSoundClassesAndModes/Foundations_Ambience
 volumeMultiplier=0.6
 concurrencyOverrides {maxCount=2, resolutionRule=PreventNew}
}
